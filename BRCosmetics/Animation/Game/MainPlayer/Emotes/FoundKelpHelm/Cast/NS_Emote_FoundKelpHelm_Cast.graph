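# Niagara effect for the cast part of the FoundKelpHelm player emote, producing cold frost puffs and ice shards.
# User.Variant gates spawn probability, switching between rate-spawned Frost sprites with IceShards meshes high above the origin and a Frost001 sprite burst near the origin.
# Frost sprites use flipbook animation, curve-driven drag and time-of-day color and alpha, while shards scale in and out under gravity.
# All emitters set their world/local position blend from an owner-speed threshold.
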NiagaraSystem /BRCosmetics/Animation/Game/MainPlayer/Emotes/FoundKelpHelm/Cast/NS_Emote_FoundKelpHelm_Cast
dependencies {
 usedDataInterfaces [&NiagaraDataInterfaceCurve]
 usedDynamicInputs []
 usedModules [&SystemState, &BlendSpaceHelper]
}
emitters {
 Frost {
  emitterUpdateScript [
   &EmitterState(Life Cycle Mode=::NewEnumerator1, Loop Behavior=::NewEnumerator1, Loop Delay=0.65, Loop Duration=0.3, UseLoopDelay=false)
   &SpawnRate(
    Spawn Probability=&MakeCustomFloatFromBool(Bool=&SetBoolByFloatComparison(A=$User.Variant, B=0.5), True Float=0, False Float=1)
    SpawnRate=10
    Use Spawn Probability=true
   )
  ]
  particleSpawnScript [
   &InitializeParticle(
    Color=&FNTimeOfDayColor(
     Night Color=&RandomRangeLinearColor(Minimum=<rgba 0.162 0.2478 0.4948 1>, Maximum=<rgba 0.3046 0.3747 0.6429 1>)
     Morning Color=&RandomRangeLinearColor(Minimum=<rgba 0.2083 0.2562 0.4948 1>, Maximum=<rgba 0.44 0.426 0.6429 1>)
     Day Color=&RandomRangeLinearColor(Minimum=<rgba 0.2083 0.2562 0.4948 1>, Maximum=<rgba 0.44 0.426 0.6429 1>)
     Evening Color=&RandomRangeLinearColor(Minimum=<rgba 0.2083 0.2562 0.4948 1>, Maximum=<rgba 0.44 0.426 0.6429 1>)
    )
    Lifetime Max=1
    Lifetime Min=1.2
    Lifetime Mode=::NewEnumerator1
    Position Offset=<0 0 165>
    Sprite Rotation Mode=::NewEnumerator2
    Sprite Size Mode=::NewEnumerator1
    Uniform Sprite Size Max=25
    Uniform Sprite Size Min=30
    UsePositionOffset=true
    Write Material Random=true
   )
   &ShapeLocation(
    Hemisphere Distribution=<1 0.5>
    MaskSpawnGroup=false
    Override Seed=false
    Sphere Radius=5
    Sphere Surface Distribution=1
    UseOwnerScale=true
   )
   &AddVelocity(
    Cone Angle=80
    Cone Axis=<0 -0.2 -1>
    Module.UseConeFalloff=false
    Module.UseVelDistribution=false
    Override Seed=false
    Velocity Mode=::NewEnumerator2
    Velocity Speed=&RandomRangeFloat(Minimum=25, Maximum=30)
   )
   &CameraOffset(Camera Offset Amount=&RandomRangeFloat(Minimum=10, Maximum=20))
  ]
  particleUpdateScript [
   &ParticleState()
   /Niagara/Modules/Update/SubUV/SubUVAnimation(Number Of Frames=64, SubUV Lookup Index=&Multiply_Float(A=$Particles.Age, B=0.66))
   &SpriteRotationRate(Rotation Rate=&RandomRangeFloat(Minimum=-35, Maximum=35))
   &Drag(
    Drag=&FloatFromCurve(
     FloatCurve=DataInterface() {propertyValues {Curve {
        defaultValue=1
        keys [{interpMode=RCIM_Cubic, tangentMode=RCTM_User, time=0.1954, value=0}, {interpMode=RCIM_Cubic, tangentMode=RCTM_User, time=0.3752, value=1}]
       }}}
     Scale Curve=&RandomRangeFloat(Minimum=3, Maximum=7)
    )
    Use Linear Drag=true
    Use Rotational Drag=true
   )
   &SolveForcesAndVelocity(Clamp Velocity=false, Limit Acceleration=false)
   &ScaleColor(
    Scale Alpha=&FloatFromCurve(
     FloatCurve=DataInterface() {propertyValues {Curve {
        defaultValue=1
        keys [
         {interpMode=RCIM_Cubic, tangentMode=RCTM_User, time=0, value=0}
         {interpMode=RCIM_Cubic, tangentMode=RCTM_User, time=0.5, value=1}
         {tangentMode=RCTM_User, time=1, value=0}
        ]
       }}}
     Scale Curve=&FNTimeOfDayScalar(Night Scale=0.1, Morning Scale=0.2, Day Scale=0.2, Evening Scale=0.2)
    )
    ScaleA=true
    ScaleRGB=false
   )
   &BlendSpace("Position Blend (World <-> Local)":NiagaraFloat=&MakeCustomFloatFromBool(
     Bool=&SetBoolByFloatComparison(A=&VectorLength(Vector=$Engine.Owner.Velocity), B=100)
     True Float=1
     False Float=0
     Translation Contribution:NiagaraFloat=1
     Rotation Contribution:NiagaraFloat=1
    ))
  ]
  rendererClasses [&NiagaraSpriteRendererProperties]
  renderers [{rendererClass=&NiagaraSpriteRendererProperties}]
 }
 Frost001 {
  emitterUpdateScript [
   &EmitterState(Life Cycle Mode=::NewEnumerator1, Loop Behavior=::NewEnumerator1, Loop Duration=1.54, UseLoopDelay=false)
   &SpawnBurst_Instantaneous(
    Spawn Count=3
    Spawn Probability=&MakeCustomFloatFromBool(Bool=&SetBoolByFloatComparison(A=$User.Variant, B=0.5), True Float=1, False Float=0)
    Use Spawn Probability=true
    UseLoopCountLimit=false
   )
  ]
  particleSpawnScript [
   &InitializeParticle(
    Color=&FNTimeOfDayColor(
     Night Color=&RandomRangeLinearColor(Minimum=<rgba 0.162 0.2478 0.4948 1>, Maximum=<rgba 0.3046 0.3747 0.6429 1>)
     Morning Color=&RandomRangeLinearColor(Minimum=<rgba 0.2083 0.2562 0.4948 1>, Maximum=<rgba 0.44 0.426 0.6429 1>)
     Day Color=&RandomRangeLinearColor(Minimum=<rgba 0.2083 0.2562 0.4948 1>, Maximum=<rgba 0.44 0.426 0.6429 1>)
     Evening Color=&RandomRangeLinearColor(Minimum=<rgba 0.2083 0.2562 0.4948 1>, Maximum=<rgba 0.44 0.426 0.6429 1>)
    )
    Lifetime Max=1
    Lifetime Min=1.54
    Lifetime Mode=::NewEnumerator1
    Position Offset=<15 10 0>
    Sprite Rotation Mode=::NewEnumerator2
    Sprite Size Mode=::NewEnumerator1
    Uniform Sprite Size Max=25
    Uniform Sprite Size Min=30
    UsePositionOffset=true
    Write Material Random=true
   )
   &ShapeLocation(MaskSpawnGroup=false, Override Seed=false, Sphere Radius=5, Sphere Surface Distribution=1, UseOwnerScale=true)
   &AddVelocity(
    Cone Angle=80
    Cone Axis=<-1 0.6 0>
    Module.UseConeFalloff=false
    Module.UseVelDistribution=false
    Override Seed=false
    Velocity Mode=::NewEnumerator2
    Velocity Speed=&RandomRangeFloat(Minimum=100, Maximum=50)
   )
   &CameraOffset(Camera Offset Amount=&RandomRangeFloat(Minimum=10, Maximum=20))
  ]
  particleUpdateScript [
   &ParticleState()
   &SubUVAnimation(Number Of Frames=64, SubUV Lookup Index=&Multiply_Float(A=$Particles.Age, B=0.66))
   &SpriteRotationRate(Rotation Rate=&RandomRangeFloat(Minimum=-35, Maximum=35))
   &Drag(
    Drag=&FloatFromCurve(
     FloatCurve=DataInterface() {propertyValues {Curve {
        defaultValue=1
        keys [
         {interpMode=RCIM_Cubic, leaveTangentWeight=0.0335, tangentMode=RCTM_User, time=0.297, value=1}
         {arriveTangentWeight=0.0335, interpMode=RCIM_Cubic, tangentMode=RCTM_User, time=0.3974, value=0}
         {arriveTangent=3.5248, interpMode=RCIM_Cubic, leaveTangent=3.5248, time=0.5226, value=0}
         <at 0.7305 1.1743 RCIM_Cubic>
        ]
       }}}
     Scale Curve=&RandomRangeFloat(Minimum=5, Maximum=4)
    )
    Use Linear Drag=true
    Use Rotational Drag=true
   )
   &SolveForcesAndVelocity(Clamp Velocity=false, Limit Acceleration=false)
   &ScaleColor(
    Scale Alpha=&FloatFromCurve(
     FloatCurve=DataInterface() {propertyValues {Curve {
        defaultValue=1
        keys [
         {interpMode=RCIM_Cubic, tangentMode=RCTM_User, time=0, value=0}
         {interpMode=RCIM_Cubic, tangentMode=RCTM_User, time=0.5, value=1}
         {tangentMode=RCTM_User, time=1, value=0}
        ]
       }}}
     Scale Curve=&FNTimeOfDayScalar(Night Scale=0.1, Morning Scale=0.2, Day Scale=0.2, Evening Scale=0.2)
    )
    ScaleA=true
    ScaleRGB=false
   )
   &BlendSpace("Position Blend (World <-> Local)":NiagaraFloat=&MakeCustomFloatFromBool(
     Bool=&SetBoolByFloatComparison(A=&VectorLength(Vector=$Engine.Owner.Velocity), B=100)
     True Float=1
     False Float=0
     Translation Contribution:NiagaraFloat=1
     Rotation Contribution:NiagaraFloat=1
    ))
  ]
  rendererClasses [&NiagaraSpriteRendererProperties]
  renderers [{rendererClass=&NiagaraSpriteRendererProperties}]
 }
 IceShards {
  emitterUpdateScript [
   &EmitterState(Life Cycle Mode=::NewEnumerator1, Loop Behavior=::NewEnumerator1, Loop Delay=0.65, Loop Duration=0.3, UseLoopDelay=false)
   &SpawnRate(
    Spawn Probability=&MakeCustomFloatFromBool(Bool=&SetBoolByFloatComparison(A=$User.Variant, B=0.5), True Float=0, False Float=1)
    SpawnRate=25
    Use Spawn Probability=true
   )
  ]
  particleSpawnScript [
   &InitializeParticle(
    Color Mode=::NewEnumerator0
    Lifetime Max=0.3
    Lifetime Min=0.25
    Lifetime Mode=::NewEnumerator1
    Mesh Scale=&Multiply_VectorByFloat(
     Vector=&RandomRangeVector(Minimum=<1 1 1>, Maximum=<1 2 0.5>)
     Float=&RandomRangeFloat(Minimum=0.05, Maximum=0.07)
    )
    Mesh Scale Mode=::NewEnumerator4
    Position Offset=<1 0 165>
    UsePositionOffset=true
    Write Material Random=true
   )
   &InitialMeshOrientation(Use Rotation Vector=false)
   &ShapeLocation(
    Hemisphere Distribution=<1 0.5>
    MaskSpawnGroup=false
    Override Seed=false
    Sphere Radius=5
    Sphere Surface Distribution=1
    UseOwnerScale=true
   )
   &AddVelocity(
    Cone Angle=150
    Cone Axis=<0 -0.5 0.8>
    Module.UseConeFalloff=false
    Module.UseVelDistribution=false
    Override Seed=false
    Velocity Mode=::NewEnumerator2
    Velocity Speed=&RandomRangeFloat(Minimum=100, Maximum=80)
   )
  ]
  particleUpdateScript [
   &ParticleState()
   &ScaleMeshSize(Scale Factor=&VectorFromFloat(Value=&FloatFromCurve(FloatCurve=DataInterface() {propertyValues {Curve {
         defaultValue=1
         keys [
          {arriveTangent=0.0594, interpMode=RCIM_Cubic, leaveTangent=0.0594, tangentMode=RCTM_User, time=0, value=0}
          {interpMode=RCIM_Cubic, tangentMode=RCTM_User, time=0.5, value=1}
          {arriveTangent=0.0205, interpMode=RCIM_Cubic, leaveTangent=0.0205, tangentMode=RCTM_User, time=1, value=0}
         ]
        }}})))
   &GravityForce()
   &Drag(Drag=&RandomRangeFloat(Minimum=2, Maximum=1), Use Linear Drag=true, Use Rotational Drag=true)
   &SolveForcesAndVelocity(Clamp Velocity=false, Limit Acceleration=false)
   &BlendSpace("Position Blend (World <-> Local)":NiagaraFloat=&MakeCustomFloatFromBool(
     Bool=&SetBoolByFloatComparison(A=&VectorLength(Vector=$Engine.Owner.Velocity), B=100)
     True Float=1
     False Float=0
     Translation Contribution:NiagaraFloat=1
     Rotation Contribution:NiagaraFloat=1
    ))
  ]
  rendererClasses [&NiagaraMeshRendererProperties]
  renderers [{rendererClass=&NiagaraMeshRendererProperties, rendererIndex=0}]
 }
}
summary {
 systemName=NS_Emote_FoundKelpHelm_Cast
 emitters [
  {emitterName=IceShards, rendererClasses [&NiagaraMeshRendererProperties]}
  {emitterName=Frost, rendererClasses [&NiagaraSpriteRendererProperties]}
  {emitterName=Frost001, rendererClasses [&NiagaraSpriteRendererProperties]}
 ]
 userVariables [User.Variant {type=NiagaraFloat, defaultValue=0, description=""}]
}
